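# Niagara system for the PistolCanary Ashen Magnus hero attack, showing repeated flashes, bullet streaks and text sprites.
# Flash_Elements rapidly bursts while AttackActive is set, placing flares, stars, spikes, meshes and lights around skeletal-mesh sockets in a ring narrowing with AttackProgress.
# Other emitters read those particles: Minimal lerps bullets from BulletSource to sampled positions, and Main_Text adds side-offset sprites.
# Drops, SpikeyFlash and Spline_Streaks are disabled.
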NiagaraSystem /PistolCanary/Assets/FX/HeroAttackAshenMagnus/NS_PistolCanary_AshenMagnus_Attack
dependencies {
 usedDataInterfaces [
  &NiagaraDataInterfaceCurve
  &NiagaraDataInterfaceColorCurve
  &NiagaraDataInterfaceParticleRead
  &NiagaraDataInterfaceArrayColor
  &NiagaraDataInterfaceArrayFloat
  &NiagaraDataInterfaceVectorCurve
  &NiagaraDataInterfaceSkeletalMesh
 ]
 usedDynamicInputs []
 usedModules [
  &SystemState
  &self:SystemSpawnScript.SystemScriptSource.SystemScriptGraph.NiagaraNodeAssignment_0.SetVariables_#1
 ]
}
emitters {
 Drops {
  bEnabled=false
  emitterUpdateScript [
   &EmitterState()
   &SpawnParticlesFromOtherEmitter(
    Attribute Reader=DataInterface() {propertyValues {EmitterBinding {bindingMode=Other, emitterName=Flash_Elements}}}
    Use Maximum Age=false
    Use Minimum Age=false
   )
  ]
  particleSpawnScript [
   &KillParticles(Kill Particles=&SetBoolByFloatComparison(
     A=&RandomRangeFloat(Minimum=&Lerp_Float(A=0, B=0.5, Alpha=$Particles.ImpactScale), Maximum=1)
     B=0.75
     Comparison Type=::NewEnumerator6
    ))
   &InitializeParticle(
    Color=<rgba 1.5371 0.0687 2 1>
    Lifetime Max=1
    Lifetime Min=0.25
    Lifetime Mode=::NewEnumerator1
    Position=$Particles.Position
    Position Mode=::NewEnumerator1
    Sprite Size Mode=::NewEnumerator3
    Uniform Sprite Size=&Multiply_Float(A=&RandomRangeFloat(Minimum=15, Maximum=25), B=3)
    UsePositionOffset=false
    Write Material Random=true
   )
   &ShapeLocation(MaskSpawnGroup=false, Override Seed=false, Sphere Radius=10, UseOwnerScale=true)
   &AddVelocity(
    Cone Angle=70
    Cone Axis=$Particles.ImpactSurfaceNormal
    Module.UseConeFalloff=false
    Module.UseVelDistribution=false
    Override Seed=false
    Velocity Mode=::NewEnumerator2
    Velocity Speed=&Lerp_Float(
     A=2500
     B=20000
     Alpha=&FloatFromCurve(
      FloatCurve=DataInterface() {propertyValues {Curve {keys [
          {arriveTangent=0.2237, interpMode=RCIM_Cubic, leaveTangent=0.2237, tangentMode=RCTM_User, time=0, value=0}
          {arriveTangent=2.6808, interpMode=RCIM_Cubic, leaveTangent=2.6808, tangentMode=RCTM_User, time=1, value=1}
         ]}}}
      CurveIndex=&RandomRangeFloat(Minimum=0, Maximum=1)
      Scale Curve=&Lerp_Float(A=0.5, B=1, Alpha=$Particles.ImpactScale)
     )
    )
   )
   &SampleParticlesFromOtherEmitter()
  ]
  particleUpdateScript [
   &ParticleState()
   &GravityForce(Gravity=<0 0 -1200>)
   &Drag(
    Drag=&FloatFromCurve(
     FloatCurve=DataInterface() {propertyValues {Curve {keys [
         {interpMode=RCIM_Cubic, leaveTangent=-8.9049, leaveTangentWeight=0.5018, tangentMode=RCTM_Break, time=0, value=1}
         {interpMode=RCIM_Cubic, tangentMode=RCTM_Break, time=0.3, value=0.05}
        ]}}}
     Scale Curve=7
    )
    Use Linear Drag=true
    Use Rotational Drag=false
   )
   &SolveForcesAndVelocity(Clamp Velocity=false, Limit Acceleration=false)
   &ScaleSpriteSize(Uniform Curve Sprite Scale=DataInterface() {propertyValues {Curve {keys [
        {interpMode=RCIM_Cubic, tangentMode=RCTM_Break, time=0.4, value=1}
        {arriveTangent=-3.9054, interpMode=RCIM_Cubic, tangentMode=RCTM_Break, time=1, value=0}
       ]}}})
   &ScaleSpriteSizeBySpeed(Max Scale Factor=<0.7 10>, Velocity Threshold=2000)
   &ScaleColor(
    Scale Alpha=&FloatFromCurve(FloatCurve=DataInterface() {propertyValues {Curve {keys [<at 0 1 RCIM_Cubic>, {arriveTangent=-2.9021, tangentMode=RCTM_User, time=1, value=0}]}}})
    ScaleA=true
    ScaleRGB=false
   ) {module=ScaleColor001}
   &DynamicMaterialParameters(
    Index 0 Param 1=&FloatFromCurve(FloatCurve=DataInterface() {propertyValues {Curve {defaultValue=1, keys [<at 0 0 RCIM_Cubic>, {arriveTangent=1.1412, tangentMode=RCTM_User, time=1, value=1}]}}})
    Index 0 Param 2=500
    Index 0 Param 3=1500
    Index 0 Param 4=200
    Module.Index0Float0Write=true
    Module.Index0Float1Write=false
    Module.Index0Float2Write=true
    Module.Index0Float3Write=false
   )
  ]
  rendererClasses [&NiagaraSpriteRendererProperties]
  renderers [{rendererClass=&NiagaraSpriteRendererProperties}]
 }
 Flash_Elements {
  emitterSpawnScript [/SetVariables(Emitter.DebugTimeMultiplier=1) {isSetParameters=true}]
  emitterUpdateScript [
   &EmitterState(Life Cycle Mode=::NewEnumerator1, Loop Duration=0.04, UseLoopDelay=true)
   &SpawnBurst_Instantaneous(Spawn Group=1, Spawn Probability=$User.AttackActive, Use Spawn Probability=true, UseLoopCountLimit=false) {module=SpawnBurst_Instantaneous001}
   &SpawnBurst_Instantaneous(
    Spawn Count=&RandomRangeInt(Minimum=6, Maximum=8)
    Use Spawn Probability=false
    UseLoopCountLimit=false
   ) {enabled=false}
  ]
  particleSpawnScript [
   &InitializeParticle(
    Color Mode=::NewEnumerator0
    Lifetime=0.25
    Mesh Scale Mode=::NewEnumerator3
    Mesh Uniform Scale=100
    UsePositionOffset=false
    Write Material Random=true
   )
   &ShapeLocation(MaskSpawnGroup=false, Override Seed=false, Sphere Radius=1, UseOwnerScale=true)
   &SpriteFacingAndAlignment(Sprite Alignment=$Particles.ShapeLocation.ShapeNormal, WriteAlignment=true, WriteFacing=false)
   &FN_PerParticleLOD(Enable SpriteSize Distance Scaling=true) {enabled=false}
   /SetVariables(Particles.SpriteSize=<1 1>, Particles.VisibilityTag=&SpawnGroup()) {isSetParameters=true}
   /SetVariables(
    Particles.Flash Spike Color:LinearColor=&MergeHelper_ColorSwitch(
     0=<rgba 0.0053 0.0208 0 1>
     1=&RandomRangeLinearColor(Minimum=<rgba 3 2.5831 1.9219 1>, Maximum=<rgba 0 2.8803 3 1>)
     2=?
     3=?
     Index=&MakeIntFromFloat(Float=$User.Variant)
     Num=::NewEnumerator0
    )
    Particles.Flash Spike Lifetime:NiagaraFloat=&RandomRangeFloat(Minimum=0.05, Maximum=0.07)
    Particles.Flash Spike Sprite Size:Vector2f=&Multiply_Vector2D(
     A=&RandomRangeVector2D(Minimum=<6 110>, Maximum=<20 140>)
     B=&Vector2DFromFloat(Value=$User.Scale)
    )
   ) {isSetParameters=true}
   /SetVariables(
    Particles.Flash Star Cam Offset=15
    Particles.Flash Star Lifetime=0.15
    Particles.Flash Star Rotation:NiagaraFloat=&RandomRangeFloat(Minimum=0, Maximum=360)
    Particles.Flash Star Sprite Size:Vector2f=&Multiply_Vector2D_ByFloat(
     Float=&Multiply_Float(A=$User.Scale, B=&RandomRangeFloat(Minimum=80, Maximum=100))
     Vector2D=$Particles.SpriteSize
    )
    Particles.Flash Star Sprite Size Alt:Vector2f=&Multiply_Vector2D_ByFloat(
     Float=&Multiply_Float(A=$User.Scale, B=&RandomRangeFloat(Minimum=53.3333, Maximum=66.6667))
     Vector2D=$Particles.SpriteSize
    )
    Particles.Flash Star SubUV Index:NiagaraFloat=&RandomRangeFloat(Minimum=0, Maximum=3)
    Particles.Flash Star SubUV Index Alt:NiagaraFloat=&RandomRangeFloat(Minimum=0, Maximum=3)
   ) {isSetParameters=true}
   /SetVariables(
    Particles.Flash Flare Lifetime:NiagaraFloat=&FN_FrontEndFloatModifier(Gameplay=0.2, Front End=0.125)
    Particles.Flash Flare Sprite Size:Vector2f=&Multiply_Vector2D(A=<130 130>, B=&Vector2DFromFloat(Value=$User.Scale))
   ) {isSetParameters=true}
   &DynamicMaterialParameters(
    Index 0 Param 1=0.2
    Index 0 Param 2=1
    Module.Index0Param0Write=true
    Module.Index0Param1Write=true
    Module.Index0Param2Write=true
    Module.Index0Param3Write=true
   )
   &SkeletalMeshLocation(
    Bone / Socket Index=&MakeIntFromFloat(Float=&FloatFromCurve(FloatCurve=DataInterface() {propertyValues {Curve {keys [<at 0 3>,<at 1 17>]}}}, CurveIndex=$System.NormalizedLoopAge))
    Bone Sampling Mode=::NewEnumerator4
    MaskSpawnGroup=false
    Skeletal Mesh=$System.Skeletal Mesh
   )
   &ShapeLocation(
    MaskSpawnGroup=false
    Override Seed=false
    Ring Radius=&FloatFromCurve(
     FloatCurve=DataInterface() {propertyValues {Curve {keys [<at 0 1 RCIM_Cubic>, {arriveTangent=-1.94, interpMode=RCIM_Cubic, leaveTangent=-1.94, tangentMode=RCTM_User, time=1, value=0.1}]}}}
     CurveIndex=$User.AttackProgress
     Scale Curve=1500
    )
    Shape Origin=&AddVectorToPosition(
     Position=$StackContext.Position
     Vector=&MakeVector(X=0, Y=0, Z=&RandomRangeFloat(Minimum=-5000, Maximum=1000))
    )
    Shape Primitive=::NewEnumerator4
    U Distribution=0.7738
    UseOwnerScale=true
   ) {module=ShapeLocation001}
   &InitialMeshOrientation(Mesh Orientation Mode=::NewEnumerator2, Orientation Vector=$Particles.ShapeLocation001.ShapeTangent, Use Rotation Vector=false)
   &CameraOffset(Camera Offset Amount=5000)
  ]
  particleUpdateScript [
   &ParticleState()
   /SetVariables(
    Particles.Update Flash Flare Age:NiagaraFloat=&Divide_Float(A=$Particles.Age, B=$Particles.Initial.Flash Flare Lifetime)
    Particles.Update Flash Spike Age:NiagaraFloat=&Divide_Float(A=$Particles.Age, B=$Particles.Initial.Flash Spike Lifetime)
    Particles.Update Flash Star Age:NiagaraFloat=&Divide_Float(A=$Particles.Age, B=$Particles.Initial.Flash Star Lifetime)
   ) {isSetParameters=true}
   /SetVariables(
    Particles.Update Flash Spike Color:LinearColor=&MakeLinearColorFromVectorAndFloat(
     "Vector (RGB)"=&MakeVectorFromLinearColorRGB(Color=$Particles.Initial.Flash Spike Color)
     "Float (Alpha)"=&FloatFromCurve(FloatCurve=DataInterface() {propertyValues {Curve {keys [<at 0.7 1>,<at 1 0>]}}}, CurveIndex=$Particles.Update Flash Spike Age)
    )
    Particles.Update Flash Spike Sprite Size:Vector2f=&Multiply_Vector2D_ByFloat(
     Float=&FloatFromCurve(FloatCurve=DataInterface() {propertyValues {Curve {keys [<at 0 0.6>,<at 1 2>]}}}, CurveIndex=$Particles.Update Flash Spike Age)
     Vector2D=$Particles.Initial.Flash Spike Sprite Size
    )
    Particles.Update Flash Star Color:LinearColor=&MergeHelper_ColorSwitch(
     0=&ColorFromCurve(
      ColorCurve=DataInterface() {propertyValues {
        AlphaCurve {keys [<at 0 1>,<at 0.8992 0>]}
        BlueCurve {keys [<at 0.0713 0.1546>,<at 0.3381 0.1097>,<at 0.6612 0.1005>]}
        GreenCurve {keys [<at 0.0713 2.1671>,<at 0.3381 1.0492>,<at 0.6612 0.5531>]}
        RedCurve {keys [<at 0.0713 3>,<at 0.3381 2>,<at 0.6612 1.5>]}
       }}
      CurveIndex=$Particles.Update Flash Star Age
     )
     1=&ColorFromCurve(
      ColorCurve=DataInterface() {propertyValues {
        AlphaCurve {keys [<at 0.2203 1>,<at 0.8992 0>]}
        BlueCurve {keys [<at 0.2249 1.1613>,<at 0.3183 0.1502>,<at 0.5354 0.0386>,<at 0.7204 0>]}
        GreenCurve {keys [<at 0.2249 3>,<at 0.3183 1.4786>,<at 0.5354 0.0368>,<at 0.7204 0.0016>]}
        RedCurve {keys [<at 0.2249 0>,<at 0.3183 3>,<at 0.5354 1.4141>,<at 0.7204 0.3672>]}
       }}
      CurveIndex=$Particles.Update Flash Star Age
     )
     2=?
     3=?
     Index=&MakeIntFromFloat(Float=$User.Variant)
     Num=::NewEnumerator0
    )
    Particles.Update Flash Star Sprize Size:Vector2f=&Multiply_Vector2D_ByFloat(
     Float=&FloatFromCurve(
      FloatCurve=DataInterface() {propertyValues {Curve {keys [
          {arriveTangent=2.7013, interpMode=RCIM_Cubic, leaveTangent=2.7013, tangentMode=RCTM_User, time=0, value=0}
          {interpMode=RCIM_Cubic, tangentMode=RCTM_User, time=1, value=1}
         ]}}}
      CurveIndex=$Particles.Update Flash Star Age
     )
     Vector2D=$Particles.Initial.Flash Star Sprite Size
    )
    Particles.Update Flash Star Sprize Size Alt:Vector2f=&Multiply_Vector2D_ByFloat(
     Float=&FloatFromCurve(
      FloatCurve=DataInterface() {propertyValues {Curve {keys [
          {arriveTangent=2.7013, interpMode=RCIM_Cubic, leaveTangent=2.7013, tangentMode=RCTM_User, time=0, value=0}
          {interpMode=RCIM_Cubic, tangentMode=RCTM_User, time=1, value=1}
         ]}}}
      CurveIndex=$Particles.Update Flash Star Age
     )
     Vector2D=$Particles.Initial.Flash Star Sprite Size Alt
    )
   ) {isSetParameters=true}
   /SetVariables(Particles.Update Flash Flare Color:LinearColor=&MakeLinearColorFromVectorAndFloat(
     "Vector (RGB)"=&Multiply_Vector(
      A=&MakeVectorFromLinearColorRGB(Color=&MergeHelper_ColorSwitch(
        0=<rgba 1 0.424 0.0506 0.7>
        1=&ColorFromCurve(ColorCurve=DataInterface() {propertyValues {
           AlphaCurve {keys [<at 0 1 RCIM_Cubic>]}
           BlueCurve {keys [<at 0 0.0177 RCIM_Cubic>,<at 0.6175 0.0354>]}
           GreenCurve {keys [<at 0 0.5984 RCIM_Cubic>,<at 0.6175 0.3469>]}
           RedCurve {keys [<at 0 1 RCIM_Cubic>,<at 0.6175 1>]}
          }})
        2=?
        3=?
        Index=&MakeIntFromFloat(Float=$User.Variant)
        Num=::NewEnumerator0
       ))
      B=&VectorFromFloat(Value=&FN_FrontEndFloatModifier(Gameplay=1, Front End=1))
     )
     "Float (Alpha)"=&Multiply_Float(
      A=&MakeFloatFromLinearColor(LinearColor=<rgba 1 1 1 1>, Channel:ENiagaraLinearColor_Channels=::NewEnumerator3)
      B=&Lerp_Float(A=1, B=0, Alpha=$Particles.Update Flash Flare Age)
     )
    )) {isSetParameters=true}
   &FN_Light_Attributes(
    Color=&MergeHelper_ColorSwitch(
     0=<rgba 0.5172 1 0.0001 0.7>
     1=&ColorFromCurve(ColorCurve=DataInterface() {propertyValues {
        AlphaCurve {keys [<at 0 1 RCIM_Cubic>,<at 1 0 RCIM_Cubic>]}
        BlueCurve {keys [<at 0 0.77 RCIM_Cubic>,<at 0.3247 0.1406>]}
        GreenCurve {keys [<at 0 1 RCIM_Cubic>,<at 0.3247 0.6878>]}
        RedCurve {keys [<at 0 0 RCIM_Cubic>,<at 0.3247 1>]}
       }})
     2=?
     3=?
     Index=&MakeIntFromFloat(Float=$User.Variant)
     Num=::NewEnumerator0
    )
    Create Color Attribute=true
    Create Exponent Attribute=true
    Create Position Attribute=true
    Create Radius Attribute=true
    Exponent=5
    Position=<15 0 0>
    Radius=90
    Scale Color=&FloatFromCurve(
     FloatCurve=DataInterface() {propertyValues {Curve {keys [
         {interpMode=RCIM_Cubic, tangentMode=RCTM_User, time=0, value=1}
         {arriveTangent=-0.5426, interpMode=RCIM_Cubic, leaveTangent=-0.5426, tangentMode=RCTM_User, time=1, value=0}
        ]}}}
     Scale Curve=10
    )
   )
  ]
  rendererClasses [
   &NiagaraSpriteRendererProperties
   &NiagaraLightRendererProperties
   &NiagaraMeshRendererProperties
  ]
  renderers [
   {rendererClass=&NiagaraSpriteRendererProperties}
   {rendererClass=&NiagaraSpriteRendererProperties, rendererIndex=1}
   {rendererClass=&NiagaraSpriteRendererProperties, rendererIndex=2}
   {rendererClass=&NiagaraSpriteRendererProperties, rendererIndex=3}
   {rendererClass=&NiagaraLightRendererProperties, rendererIndex=4}
   {rendererClass=&NiagaraMeshRendererProperties, rendererIndex=5}
   {rendererClass=&NiagaraMeshRendererProperties, rendererIndex=6}
  ]
 }
 Main_Text {
  emitterUpdateScript [
   &EmitterState()
   &SpawnBurst_Instantaneous(Spawn Count=1, Use Spawn Probability=false)
   &SpawnParticlesFromOtherEmitter(
    Attribute Reader=DataInterface() {propertyValues {EmitterBinding {bindingMode=Other, emitterName=Flash_Elements}}}
    Spawn Rate=$User.TextSpawn
    Use Maximum Age=false
    Use Minimum Age=false
   )
  ]
  particleSpawnScript [
   &InitializeParticle(
    Color=&SelectLinearColorFromArray(
     Array Sampling Mode=::NewEnumerator0
     Direct Array Index=?
     Color Selection Array=DataInterface() {propertyValues {ColorData [<rgba 1 0.6769 0.2511 1>, <rgba 1 0.6758 0.0551 1>]}}
     Random Seed=0
    )
    Lifetime=0.7
    Sprite Size=&Multiply_Vector2D_ByFloat(
     Float=&Multiply_Float(
      A=$User.Size_Main
      B=&Multiply_Float(
       A=&RandomRangeFloat(Minimum=0.7, Maximum=1.1)
       B=&SelectFloatFromArray(
        Array Sampling Mode=::NewEnumerator0
        Direct Array Index=?
        Float Selection Array=DataInterface() {propertyValues {FloatData [1,1.4]}}
        Random Seed=0
       )
      )
     )
     Vector2D=<16000 10000>
    )
    Sprite Size Mode=::NewEnumerator4
    UsePositionOffset=false
    Write Material Random=true
   ) {module=InitializeParticle001}
   &CameraOffset(Camera Offset Amount=5000)
   &SampleParticlesFromOtherEmitter(Applied Position Offset:Vector3f=&MakeVector(
     X=0
     Y=&Multiply_Float(
      A=&Lerp_Float(A=2000, B=300, Alpha=$User.AttackProgress)
      B=&SelectFloatFromArray(
       Array Sampling Mode=::NewEnumerator0
       Direct Array Index=?
       Float Selection Array=DataInterface() {propertyValues {FloatData [1,-1]}}
       Random Seed=0
      )
     )
     Z=0
    ))
   &DynamicMaterialParameters(
    Index 0 Param 2=&SelectFloatFromArray(
     Array Sampling Mode=::NewEnumerator0
     Direct Array Index=?
     Float Selection Array=DataInterface() {propertyValues {FloatData [0,0.5,1,1.5]}}
     Random Seed=0
    )
    Module.Index0Float0Write=false
    Module.Index0Float1Write=true
    Module.Index0Float2Write=true
    Module.Index0Float3Write=true
   )
  ]
  particleUpdateScript [
   &ParticleState()
   &ScaleSpriteSize(
    Uniform Curve Scale=2
    Uniform Curve Sprite Scale=DataInterface() {propertyValues {Curve {keys [
        {arriveTangent=47.5236, interpMode=RCIM_Cubic, leaveTangent=47.5236, tangentMode=RCTM_User, time=0, value=0}
        {arriveTangent=0.9645, interpMode=RCIM_Cubic, leaveTangent=0.9645, tangentMode=RCTM_User, time=0.0484, value=0.9847}
        {arriveTangent=-0.0405, interpMode=RCIM_Cubic, leaveTangent=-0.0405, tangentMode=RCTM_User, time=1, value=1}
       ]}}}
   )
   &ScaleColor(
    Scale Alpha=&FloatFromCurve(FloatCurve=DataInterface() {propertyValues {Curve {keys [
         {arriveTangent=-0.0305, interpMode=RCIM_Cubic, leaveTangent=-0.0305, tangentMode=RCTM_User, time=0, value=1}
         {interpMode=RCIM_Cubic, tangentMode=RCTM_User, time=1, value=0}
        ]}}})
    ScaleA=false
    ScaleRGB=false
   )
   &SolveForcesAndVelocity(Clamp Velocity=false, Limit Acceleration=false)
   &DynamicMaterialParameters(
    Index 0 Param 1=&FloatFromCurve(FloatCurve=DataInterface() {propertyValues {Curve {keys [<at 0.6 1>,<at 1 0>]}}})
    Module.Index0Float0Write=true
    Module.Index0Float1Write=false
    Module.Index0Float2Write=false
    Module.Index0Float3Write=false
   ) {module=DynamicMaterialParameters001}
  ]
  rendererClasses [&NiagaraSpriteRendererProperties]
  renderers [{rendererClass=&NiagaraSpriteRendererProperties}]
 }
 Minimal {
  emitterUpdateScript [
   &EmitterState()
   &SpawnParticlesFromOtherEmitter(
    Attribute Reader=DataInterface() {propertyValues {EmitterBinding {bindingMode=Other, emitterName=Flash_Elements}}}
    Spawn Rate=7
    Use Maximum Age=false
    Use Minimum Age=false
   )
  ]
  particleSpawnScript [
   &InitializeParticle(
    Color Mode=::NewEnumerator0
    Lifetime=0.1
    Mesh Scale Mode=::NewEnumerator3
    Mesh Uniform Scale=200
    Sprite Size Mode=::NewEnumerator3
    Uniform Sprite Size=200
    UsePositionOffset=false
    Write Material Random=true
   )
   &SkeletalMeshLocation(
    Bone / Socket Index=&MakeIntFromFloat(Float=&FloatFromCurve(FloatCurve=DataInterface() {propertyValues {Curve {keys [<at 0 3>,<at 1 17>]}}}, CurveIndex=$System.NormalizedLoopAge))
    Bone Sampling Mode=::NewEnumerator4
    MaskSpawnGroup=false
    Orientation Sampling=::NewEnumerator1
    Position Sampling=::NewEnumerator2
    Skeletal Mesh=$System.Skeletal Mesh
   )
   &SampleParticlesFromOtherEmitter(
    Mesh Orientation Sampling:ENiagara_DirectReadSamplingMode=::NewEnumerator1
    Position Sampling:ENiagara_DirectReadSamplingMode=::NewEnumerator1
   )
   /SetVariables(
    Particles.SampledPos:NiagaraPosition=$Output.SampleParticlesFromOtherEmitter.SampledPosition
    Particles.StartingPos:Vector3f=&AddVectorToPosition(Position=$User.BulletSource, Vector=&RandomRangeVector(Minimum=<-100 100 20>, Maximum=<100 -100 1>))
   ) {isSetParameters=true}
   /SetVariables(Particles.BulletColor:LinearColor=<rgba 1 0.3549 0 1>, Particles.BulletSize:Vector3f=<20 5 5>) {isSetParameters=true}
   &InitialMeshOrientation(
    Mesh Orientation Mode=::NewEnumerator2
    Orientation Vector=&Subtract_Position(A=$User.BulletSource, B=$Output.SampleParticlesFromOtherEmitter.SampledPosition)
    Use Rotation Vector=false
   )
   &CameraOffset(Camera Offset Amount=1000)
  ]
  particleUpdateScript [
   &ParticleState()
   &SolveForcesAndVelocity(Clamp Velocity=false, Limit Acceleration=false)
   /SetVariables(Particles.Position=&Lerp_Vector(
     A=$Particles.Initial.StartingPos
     B=$Particles.Initial.SampledPos
     Alpha=&FloatFromCurve(FloatCurve=DataInterface() {propertyValues {Curve {defaultValue=1, keys [{tangentMode=RCTM_User, time=0, value=0}, {tangentMode=RCTM_User, time=1, value=1}]}}})
    )) {isSetParameters=true}
   /SetVariables(Particles.BulletSize:Vector3f=&Multiply_Vector(
     A=$Particles.Initial.BulletSize
     B=&MakeVector(
      X=&FloatFromCurve(FloatCurve=DataInterface() {propertyValues {Curve {keys [<at 0.6 1>,<at 1 0>]}}})
      Y=1
      Z=1
     )
    )) {isSetParameters=true}
  ]
  rendererClasses [&NiagaraSpriteRendererProperties, &NiagaraMeshRendererProperties]
  renderers [
   {rendererClass=&NiagaraSpriteRendererProperties}
   {rendererClass=&NiagaraMeshRendererProperties, rendererIndex=1}
  ]
 }
 SpikeyFlash {
  bEnabled=false
  emitterUpdateScript [
   &EmitterState()
   &SpawnParticlesFromOtherEmitter(
    Attribute Reader=DataInterface() {propertyValues {EmitterBinding {bindingMode=Other, emitterName=Flash_Elements}}}
    Spawn Rate=$User.FlashSpawn
    Use Maximum Age=false
    Use Minimum Age=false
   )
  ]
  particleSpawnScript [
   &InitializeParticle(
    Color=<rgba 2 0.0932 0.4976 1>
    Lifetime=&Lerp_Float(
     A=&RandomRangeFloat(Minimum=0.07, Maximum=0.085)
     B=&RandomRangeFloat(Minimum=0.07, Maximum=0.085)
     Alpha=$Particles.ImpactScale
    )
    Mesh Scale=&Multiply_VectorByFloat(
     Vector=&RandomRangeVector(Minimum=<0.1 0.1 0.5>, Maximum=<0.15 0.15 0.8>)
     Float=&Lerp_Float(A=0.5, B=50000, Alpha=$Particles.ImpactScale)
    )
    Mesh Scale Mode=::NewEnumerator4
    Position=$Particles.Position
    Position Mode=::NewEnumerator1
    Sprite Rotation Mode=::NewEnumerator2
    Sprite Size Mode=::NewEnumerator3
    Uniform Sprite Size=&Multiply_Float(A=2500, B=&Lerp_Float(A=0.25, B=1, Alpha=$Particles.ImpactScale))
    UsePositionOffset=false
    Write Material Random=true
   )
   &CameraOffset(Camera Offset Amount=50)
   &InitialMeshOrientation(
    Mesh Axis To Orient=<0 0 1>
    Mesh Orientation Mode=::NewEnumerator2
    Orientation Vector=$Particles.ImpactSurfaceNormal
    Rotation=&RandomRangeVector(Minimum=<0 0 0>, Maximum=<0 0 1>)
    Use Rotation Vector=true
   )
   &SpriteFacingAndAlignment(Sprite Facing=$Particles.ImpactSurfaceNormal, WriteAlignment=false, WriteFacing=true)
   &SampleParticlesFromOtherEmitter()
  ]
  particleUpdateScript [
   &ParticleState()
   &ScaleSpriteSize(Uniform Curve Sprite Scale=DataInterface() {propertyValues {Curve {keys [
        {arriveTangent=6.3514, interpMode=RCIM_Cubic, leaveTangent=6.3514, tangentMode=RCTM_User, time=0, value=0}
        {tangentMode=RCTM_User, time=1, value=1}
       ]}}})
   &SolveForcesAndVelocity(Clamp Velocity=false, Limit Acceleration=false)
   &ScaleMeshSize(Scale Factor=&VectorFromFloat(Value=&FloatFromCurve(
      FloatCurve=DataInterface() {propertyValues {Curve {keys [
          {arriveTangent=9.901, interpMode=RCIM_Cubic, leaveTangent=9.901, tangentMode=RCTM_User, time=0, value=0.1}
          {arriveTangent=2.6083, interpMode=RCIM_Cubic, leaveTangent=2.6083, tangentMode=RCTM_User, time=1, value=6}
         ]}}}
      Scale Curve=6
     )))
   &ScaleColor(
    Scale Alpha=&FloatFromCurve(FloatCurve=DataInterface() {propertyValues {Curve {keys [<at 0.4 1 RCIM_Cubic>, {arriveTangent=-2.2027, tangentMode=RCTM_User, time=1, value=0}]}}})
    Scale RGB=&VectorFromFloat(Value=2)
    ScaleA=true
    ScaleRGB=true
   )
   /SetVariables(Particles.Glow Sprite Color:LinearColor=&MakeLinearColorFromVectorAndFloat(
     "Vector (RGB)"=&MakeVectorFromLinearColorRGB(Color=<rgba 5 0.0781 0.9369 1>)
     "Float (Alpha)"=&FloatFromCurve(
      FloatCurve=DataInterface() {propertyValues {Curve {keys [{arriveTangent=-2.7319, interpMode=RCIM_Cubic, leaveTangent=-2.7319, tangentMode=RCTM_User, time=0, value=1}, <at 1 0>]}}}
      Scale Curve=2
     )
    )) {isSetParameters=true}
  ]
  rendererClasses [&NiagaraMeshRendererProperties, &NiagaraSpriteRendererProperties]
  renderers [
   {rendererClass=&NiagaraMeshRendererProperties, rendererIndex=0}
   {rendererClass=&NiagaraSpriteRendererProperties, rendererIndex=1}
  ]
 }
 Spline_Streaks {
  bEnabled=false
  emitterUpdateScript [
   &EmitterState()
   &SpawnParticlesFromOtherEmitter(
    Attribute Reader=DataInterface() {propertyValues {EmitterBinding {bindingMode=Other, emitterName=Flash_Elements}}}
    Use Maximum Age=false
    Use Minimum Age=false
   )
  ]
  particleSpawnScript [
   &KillParticles(Kill Particles=&SetBoolByFloatComparison(
     A=&RandomRangeFloat(Minimum=&Lerp_Float(A=0, B=0.5, Alpha=$Particles.ImpactScale), Maximum=1)
     B=0.75
     Comparison Type=::NewEnumerator6
    ))
   /SetVariables(Particles.RandomIntensity:NiagaraFloat=&FloatFromCurve(
     FloatCurve=DataInterface() {propertyValues {Curve {defaultValue=1, keys [<at 0 0 RCIM_Cubic>, {arriveTangent=1.6936, tangentMode=RCTM_User, time=1, value=1}]}}}
     CurveIndex=&RandomRangeFloat(Minimum=0, Maximum=&Lerp_Float(A=0.6, B=1, Alpha=$Particles.ImpactScale))
    )) {isSetParameters=true}
   &InitializeParticle(
    Color=<rgba 1.5371 0.0687 2 1>
    Lifetime=&Multiply_Float(
     A=&RandomRangeFloat(Minimum=0.5, Maximum=0.7)
     B=&Lerp_Float(A=0.5, B=1, Alpha=$Particles.RandomIntensity)
    )
    Mesh Scale Mode=::NewEnumerator3
    Mesh Uniform Scale=&Multiply_Float(
     A=&RandomRangeFloat(Minimum=0.7, Maximum=1)
     B=&Lerp_Float(A=1, B=6, Alpha=$Particles.RandomIntensity)
    )
    Position=$Particles.Position
    Position Mode=::NewEnumerator1
    UsePositionOffset=false
    Write Material Random=true
   )
   &ShapeLocation(
    Cylinder Height=5
    Cylinder Radius=10
    MaskSpawnGroup=false
    Override Seed=false
    Rotation Axis=<0 1 0>
    Rotation Mode=::NewEnumerator1
    Shape Primitive=::NewEnumerator1
    Surface Only=true
    Use Endcaps In Surface Only Mode=false
    UseOwnerScale=true
   ) {module=ShapeLocation001}
   &AddVelocity(
    Cone Angle=90
    Cone Axis=$Particles.ImpactSurfaceNormal
    Module.UseConeFalloff=false
    Module.UseVelDistribution=true
    Override Seed=false
    Velocity Mode=::NewEnumerator2
    Velocity Speed=&Lerp_Float(A=200, B=500, Alpha=$Particles.RandomIntensity)
   )
   &SampleParticlesFromOtherEmitter()
  ]
  particleUpdateScript [
   &ParticleState()
   &UpdateMeshOrientation(
    Rotation Rate=&FloatFromCurve(
     FloatCurve=DataInterface() {propertyValues {Curve {keys [
         {arriveTangent=-2.1925, interpMode=RCIM_Cubic, leaveTangent=-2.1925, tangentMode=RCTM_User, time=0, value=1}
         {arriveTangent=-0.2845, interpMode=RCIM_Cubic, leaveTangent=-0.2845, tangentMode=RCTM_User, time=1, value=0.05}
        ]}}}
     Scale Curve=&RandomRangeFloat(Minimum=0.15, Maximum=0.25)
    )
    Rotation Vector=<0 1 -1>
   )
   &ScaleMeshSize(Scale Factor=&VectorFromCurve(
     VectorCurve=DataInterface() {propertyValues {
       XCurve {keys [
         {arriveTangent=2.4453, interpMode=RCIM_Cubic, leaveTangent=2.4453, tangentMode=RCTM_User, time=0, value=0}
         {arriveTangent=0.1848, tangentMode=RCTM_User, time=1, value=1}
        ]}
       YCurve {keys [<at 0 1>,<at 1 0>]}
       ZCurve {keys [
         {arriveTangent=0.283, interpMode=RCIM_Cubic, leaveTangent=0.283, tangentMode=RCTM_User, time=0, value=0}
         {arriveTangent=2.0397, tangentMode=RCTM_User, time=1, value=1}
        ]}
      }}
     Scale Curve=<10 10 5>
    ))
   &GravityForce(Gravity=<0 0 -400>)
   &Drag(Drag=&RandomRangeFloat(Minimum=0.25, Maximum=0.5), Use Linear Drag=true, Use Rotational Drag=false)
   &SolveForcesAndVelocity(Clamp Velocity=false, Limit Acceleration=false)
   &DynamicMaterialParameters(
    Index 0 Param 1=&FloatFromCurve(FloatCurve=DataInterface() {propertyValues {Curve {keys [<at 0.4 0 RCIM_Cubic>, {arriveTangent=2.3043, interpMode=RCIM_Cubic, leaveTangent=2.3043, tangentMode=RCTM_User, time=1, value=1}]}}})
    Index 0 Param 2=&FloatFromCurve(
     FloatCurve=DataInterface() {propertyValues {Curve {keys [
         {arriveTangent=1.3993, interpMode=RCIM_Cubic, leaveTangent=1.3993, tangentMode=RCTM_User, time=0, value=0.5}
         {interpMode=RCIM_Cubic, tangentMode=RCTM_User, time=1, value=1}
        ]}}}
     Scale Curve=&RandomRangeFloat(Minimum=0, Maximum=2)
    )
    Index 0 Param 3=&Multiply_Float(
     A=&MakeFloatFromVector(Vector=$Particles.Initial.Scale, Channel=::NewEnumerator0)
     B=&RandomRangeFloat(Minimum=0.5, Maximum=2)
    )
    Index 0 Param 4=&FloatFromCurve(
     FloatCurve=DataInterface() {propertyValues {Curve {keys [
         {interpMode=RCIM_Cubic, leaveTangentWeight=0.7017, tangentMode=RCTM_User, time=0, value=0}
         {arriveTangent=1.7252, arriveTangentWeight=1.2352, interpMode=RCIM_Cubic, leaveTangent=1.7252, tangentMode=RCTM_User, time=1, value=1}
        ]}}}
     Scale Curve=&RandomRangeFloat(Minimum=0.75, Maximum=2)
    )
    Index 1 Param 1=&FloatFromCurve(
     FloatCurve=DataInterface() {propertyValues {Curve {keys [
         {arriveTangent=2.4162, interpMode=RCIM_Cubic, leaveTangent=2.4162, tangentMode=RCTM_User, time=0, value=0}
         {interpMode=RCIM_Cubic, tangentMode=RCTM_User, time=1, value=1}
        ]}}}
     Scale Curve=-1
    )
    Index 1 Param 2=&RandomRangeFloat(Minimum=0.75, Maximum=1.75)
    Module.Index0Float0Write=true
    Module.Index0Float1Write=true
    Module.Index0Float2Write=true
    Module.Index0Float3Write=true
    Module.Index1Float0Write=true
    Module.Index1Float1Write=true
    Module.Index1Float2Write=false
    Module.Index1Float3Write=false
    Write Parameter Index 1=::NewEnumerator1
   )
  ]
  rendererClasses [&NiagaraMeshRendererProperties]
  renderers [{rendererClass=&NiagaraMeshRendererProperties, rendererIndex=0}]
 }
}
summary {
 systemName=NS_PistolCanary_AshenMagnus_Attack
 emitters [
  {
   emitterName=Flash_Elements
   rendererClasses [
    &NiagaraSpriteRendererProperties
    &NiagaraLightRendererProperties
    &NiagaraMeshRendererProperties
   ]
  }
  {emitterName=Minimal, rendererClasses [&NiagaraSpriteRendererProperties, &NiagaraMeshRendererProperties]}
  {emitterName=Main_Text, rendererClasses [&NiagaraSpriteRendererProperties]}
  {
   bEnabled=false
   emitterName=SpikeyFlash
   rendererClasses [&NiagaraMeshRendererProperties, &NiagaraSpriteRendererProperties]
  }
  {bEnabled=false, emitterName=Drops, rendererClasses [&NiagaraSpriteRendererProperties]}
  {bEnabled=false, emitterName=Spline_Streaks, rendererClasses [&NiagaraMeshRendererProperties]}
 ]
 userVariables [
  User.AttackActive {type=NiagaraFloat, defaultValue=1, description=""}
  User.AttackProgress {type=NiagaraFloat, defaultValue=0.4, description=""}
  User.BeamEnd_0 {type=Vector3f, description="", defaultValue {struct=Vector3f, value=<0 0 0>}}
  User.bIsFrontEnd {type=NiagaraBool}
  User.bIsFrontEndPreview {type=NiagaraBool}
  User.BulletSource {type=NiagaraPosition, description="", defaultValue {}}
  User.Duration {type=NiagaraFloat, defaultValue=4, description=""}
  User.FlashSpawn {type=NiagaraFloat, defaultValue=0, description=""}
  User.MaxDistance {type=NiagaraFloat, defaultValue=0, description=""}
  User.PhysicalSurface {type=NiagaraInt32}
  User.Scale {type=NiagaraFloat, defaultValue=10, description=""}
  User.Size_Main {type=NiagaraFloat, defaultValue=0.04, description=""}
  User.TextSpawn {type=NiagaraFloat, defaultValue=0, description=""}
  User.Variant {type=NiagaraFloat, defaultValue=0, description=""}
  User.Variantx {type=NiagaraInt32}
 ]
}
